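# Physics asset for the hood mesh of the female medium FairyFlex player body, providing collision bodies and joints for its dyn_cape bones.
# Single-capsule bodies on spine_05, dyn_cape_master, dyn_cape_01 and dyn_cape_02 form a constrained chain, with a large capsule at the upper spine.
# The master-to-spine joint swings and twists freely, while the two downstream joints have narrow, asymmetric swing limits and zero twist allowance.
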PhysicsAsset /BRCosmetics/Characters/Player/Female/Medium/Bodies/F_MED_FairyFlex/Meshes/F_MED_FairyFlex_Hood_Physics
bodies {
 dyn_cape_01 [{shapeType=Capsule, length=9.1003, radius=3.5561, shapeName=dyn_cape_01_capsule, center=<0.7157 -2 0>}]
 dyn_cape_02 [{shapeType=Capsule, length=5.2041, radius=2.2333, shapeName=dyn_cape_02_capsule, center=<2.0031 0 0>}]
 dyn_cape_master [{shapeType=Capsule, length=0.505, radius=0.505, shapeName=dyn_cape_master_capsule, rotation=<rot 0 -90 0>}]
 spine_05 [{shapeType=Capsule, length=13.1957, radius=12.6868, shapeName=spine_05_capsule, center=<0 0.5636 0>}]
}
constraints [
 {
  bone1Name=dyn_cape_master
  bone2Name=spine_05
  swing1LimitDegrees=45
  swing1Motion=Free
  swing2LimitDegrees=45
  swing2Motion=Free
  twistLimitDegrees=45
  twistMotion=Free
 }
 {bone1Name=dyn_cape_02, bone2Name=dyn_cape_01, swing1LimitDegrees=25, swing2LimitDegrees=10, twistLimitDegrees=0}
 {bone1Name=dyn_cape_01, bone2Name=dyn_cape_master, swing1LimitDegrees=25, swing2LimitDegrees=10, twistLimitDegrees=0}
]
