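# Physics asset for the key and back keychain accessory on the F_MED_FairyFlex female character body, using capsule bodies for the dangling dyn_ bones.
# The key is a two-segment chain under a master body locked angularly to the pelvis, while the keychain is a separate three-link chain from the pelvis.
# Both chains have narrow, asymmetric swing limits and zero twist.
# Pelvis and right thigh capsules supply body volumes, joined by an angularly free constraint.
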PhysicsAsset /BRCosmetics/Characters/Player/Female/Medium/Bodies/F_MED_FairyFlex/Meshes/F_MED_FairyFlex_Key_Physics
bodies {
 dyn_key_01 [{
   shapeType=Capsule
   length=5.3548
   radius=1.168
   shapeName=dyn_key_01_capsule
   center=<0.8684 -0.2769 -0.1192>
   rotation=<rot -6.8085 -0.0001 0>
  }]
 dyn_key_02 [{
   shapeType=Capsule
   length=12.3046
   radius=1.7432
   shapeName=dyn_key_02_capsule
   center=<7.0558 -0.0818 -0.7125>
   rotation=<rot 0 -90.327 -90.5019>
  }]
 dyn_key_Master [{shapeType=Capsule, length=0.505, radius=0.505, shapeName=dyn_key_Master_capsule, rotation=<rot 0 -90 0>}]
 dyn_keychain_back_01 [{shapeType=Capsule, length=3, radius=2, shapeName=dyn_keychain_back_01_capsule}]
 dyn_keychain_back_02 [{shapeType=Capsule, length=3, radius=2, shapeName=dyn_keychain_back_02_capsule}]
 dyn_keychain_back_03 [{shapeType=Capsule, length=3, radius=3.1785, shapeName=dyn_keychain_back_03_capsule, center=<3.9086 -2.3892 0>}]
 pelvis [
  {shapeType=Capsule, length=14.19, radius=8.8314, shapeName=pelvis_capsule, center=<-2.601 -2.7302 0.0856>, rotation=<rot 4.8161 -90 1.1>}
  {shapeType=Capsule, length=5.7215, radius=10.39, shapeName=pelvis_capsule, center=<-6.6817 -2.8207 1.5521>}
 ]
 thigh_r [{
   shapeType=Capsule
   length=14.2383
   radius=8.1844
   shapeName=thigh_r_capsule
   center=<8.8725 1.4598 -0.1008>
   rotation=<rot 90 -176.4237 -166.2213>
  }]
}
constraints [
 {bone1Name=dyn_key_02, bone2Name=dyn_key_01, swing1LimitDegrees=1, swing2LimitDegrees=15, twistLimitDegrees=0}
 {bone1Name=dyn_key_01, bone2Name=dyn_key_Master, swing1LimitDegrees=0, swing2LimitDegrees=11.64, twistLimitDegrees=0}
 {
  bone1Name=thigh_r
  bone2Name=pelvis
  swing1LimitDegrees=45
  swing1Motion=Free
  swing2LimitDegrees=45
  swing2Motion=Free
  twistLimitDegrees=45
  twistMotion=Free
 }
 {
  bone1Name=dyn_key_Master
  bone2Name=pelvis
  swing1LimitDegrees=45
  swing1Motion=Locked
  swing2LimitDegrees=45
  swing2Motion=Locked
  twistLimitDegrees=45
  twistMotion=Locked
 }
 {bone1Name=dyn_keychain_back_01, bone2Name=pelvis, swing1LimitDegrees=20, swing2LimitDegrees=3, twistLimitDegrees=0}
 {bone1Name=dyn_keychain_back_02, bone2Name=dyn_keychain_back_01, swing1LimitDegrees=20, swing2LimitDegrees=3, twistLimitDegrees=0}
 {bone1Name=dyn_keychain_back_03, bone2Name=dyn_keychain_back_02, swing1LimitDegrees=15, swing2LimitDegrees=3, twistLimitDegrees=0}
]
