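# Physics asset for the FairyFlex female medium body cosmetic, simulating three hanging rag strips.
# Each strip is a three-capsule chain on a small master body constrained to the pelvis with free rotation.
# Chain joints allow narrow swing on one axis, wider on the other, and almost no twist.
# Pelvis and right thigh capsules plus a wide, thin root box provide body and apparent floor shapes.
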PhysicsAsset /BRCosmetics/Characters/Player/Female/Medium/Bodies/F_MED_FairyFlex/Meshes/F_MED_FairyFlex_PhysicsAsset
bodies {
 dyn_rag_01_01 [{shapeType=Capsule, length=3, radius=3, shapeName=dyn_rag_01_01_capsule, center=<3.3403 0 2.4931>, rotation=<rot 0 0 90>}]
 dyn_rag_01_02 [{shapeType=Capsule, length=3, radius=3, shapeName=dyn_rag_01_02_capsule, center=<3.3356 -0.0003 1.7>, rotation=<rot 0 0 90.0748>}]
 dyn_rag_01_03 [{shapeType=Capsule, length=13.12, radius=3, shapeName=dyn_rag_01_03_capsule, center=<5.76 -0.1252 1>, rotation=<rot 0.005 -5.19 90.0013>}]
 dyn_rag_02_01 [{shapeType=Capsule, length=3, radius=3, shapeName=dyn_rag_02_01_capsule, center=<3.3356 0 3>, rotation=<rot 0 0 90>}]
 dyn_rag_02_02 [{shapeType=Capsule, length=5.8404, radius=3, shapeName=dyn_rag_02_02_capsule, center=<4.4475 0 1.7>, rotation=<rot 0 0 90.0486>}]
 dyn_rag_02_03 [{shapeType=Capsule, length=12.0646, radius=3, shapeName=dyn_rag_02_03_capsule, center=<5.9202 0 0>, rotation=<rot 0 0 90>}]
 dyn_rag_03_01 [{shapeType=Capsule, length=3, radius=3, shapeName=dyn_rag_03_01_capsule, center=<3.3356 0 2>, rotation=<rot 0 0 90>}]
 dyn_rag_03_02 [{shapeType=Capsule, length=3, radius=3, shapeName=dyn_rag_03_02_capsule, center=<4.4227 0 2>, rotation=<rot 0 0 90>}]
 dyn_rag_03_03 [{shapeType=Capsule, length=12.282, radius=3, shapeName=dyn_rag_03_03_capsule, center=<4.2833 0 0.7>, rotation=<rot -0.0001 -0.0002 94.0561>}]
 dyn_rag_master_01 [{shapeType=Capsule, length=0.505, radius=0.505, shapeName=dyn_rag_master_01_capsule, rotation=<rot 0 -90 0>}]
 dyn_rag_master_02 [{shapeType=Capsule, length=0.505, radius=0.505, shapeName=dyn_rag_master_02_capsule, rotation=<rot 0 -90 0>}]
 dyn_rag_master_03 [{shapeType=Capsule, length=0.505, radius=0.505, shapeName=dyn_rag_master_03_capsule, rotation=<rot 0 -90 0>}]
 pelvis [{shapeType=Capsule, length=12.7026, radius=8.8028, shapeName=pelvis_capsule, center=<-5.4265 -4.4707 0>, rotation=<rot 4.8934 -90 1.1736>}]
 root [{shapeType=Box, extentX=200, extentY=200, extentZ=1, shapeName=root_box}]
 thigh_r [{
   shapeType=Capsule
   length=15.5206
   radius=8.837
   shapeName=thigh_r_capsule
   center=<11.8862 0.8537 -0.0149>
   rotation=<rot 84.2373 172.4456 179.9999>
  }]
}
constraints [
 {
  bone1Name=thigh_r
  bone2Name=pelvis
  swing1LimitDegrees=45
  swing1Motion=Free
  swing2LimitDegrees=45
  swing2Motion=Free
  twistLimitDegrees=45
  twistMotion=Free
 }
 {
  bone1Name=dyn_rag_master_01
  bone2Name=pelvis
  swing1LimitDegrees=45
  swing1Motion=Free
  swing2LimitDegrees=45
  swing2Motion=Free
  twistLimitDegrees=45
  twistMotion=Free
 }
 {bone1Name=dyn_rag_01_01, bone2Name=dyn_rag_master_01, swing1LimitDegrees=3, swing2LimitDegrees=15, twistLimitDegrees=0}
 {bone1Name=dyn_rag_01_02, bone2Name=dyn_rag_01_01, swing1LimitDegrees=3, swing2LimitDegrees=20, twistLimitDegrees=0}
 {bone1Name=dyn_rag_01_03, bone2Name=dyn_rag_01_02, swing1LimitDegrees=3, swing2LimitDegrees=20, twistLimitDegrees=3}
 {
  bone1Name=dyn_rag_master_02
  bone2Name=pelvis
  swing1LimitDegrees=45
  swing1Motion=Free
  swing2LimitDegrees=45
  swing2Motion=Free
  twistLimitDegrees=45
  twistMotion=Free
 }
 {bone1Name=dyn_rag_02_01, bone2Name=dyn_rag_master_02, swing1LimitDegrees=3, swing2LimitDegrees=20, twistLimitDegrees=0}
 {bone1Name=dyn_rag_02_02, bone2Name=dyn_rag_02_01, swing1LimitDegrees=3, swing2LimitDegrees=25, twistLimitDegrees=0}
 {bone1Name=dyn_rag_02_03, bone2Name=dyn_rag_02_02, swing1LimitDegrees=3, swing2LimitDegrees=25, twistLimitDegrees=0}
 {
  bone1Name=dyn_rag_master_03
  bone2Name=pelvis
  swing1LimitDegrees=45
  swing1Motion=Free
  swing2LimitDegrees=45
  swing2Motion=Free
  twistLimitDegrees=45
  twistMotion=Free
 }
 {bone1Name=dyn_rag_03_01, bone2Name=dyn_rag_master_03, swing1LimitDegrees=3, swing2LimitDegrees=20, twistLimitDegrees=0}
 {bone1Name=dyn_rag_03_02, bone2Name=dyn_rag_03_01, swing1LimitDegrees=3, swing2LimitDegrees=25, twistLimitDegrees=0}
 {bone1Name=dyn_rag_03_03, bone2Name=dyn_rag_03_02, swing1LimitDegrees=3, swing2LimitDegrees=25, twistLimitDegrees=0}
 {
  bone1Name=pelvis
  bone2Name=root
  swing1LimitDegrees=45
  swing1Motion=Free
  swing2LimitDegrees=45
  swing2Motion=Free
  twistLimitDegrees=45
  twistMotion=Free
 }
]
