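# Control Rig for the FairyFlex female player cosmetic's face accessory that applies corrective offsets to back-hair base bones as the head tilts.
# It remaps head rotation relative to spine_05 into three staged "Head Up" alphas and pelvis height into a crouch alpha.
# Products and an inverse of these alphas weight four Modify Transforms passes, two smoothed by AlphaInterp, one crouch-dependent.
# Head Right is computed but never read.
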ControlRigBlueprint /BRCosmetics/Characters/Player/Female/Medium/Bodies/F_MED_FairyFlex/Meshes/Parts/F_MED_FairyFlex_FaceAcc_CtrlRig
graphs [
 RigVMFunctionLibrary {links [], nodes []}
 RigVMModel {nodes [
   RigUnit_BeginExecution = &RigUnit_BeginExecution()
   RigVMVariableNode(
    Value=&RigVMFunction_MathDoubleRemap(
     SourceMaximum=50
     SourceMinimum=60
     TargetMaximum=1
     TargetMinimum=0
     Value=&RigUnit_GetRelativeTransformForItem(Child=(Type=Bone,Name="pelvis"), Parent=(Type=Bone,Name="root"), bChildInitial=False, bParentInitial=False).RelativeTransform.Translation.Z
    ).Result
    ExecuteContext=@RigUnit_BeginExecution.ExecutePin
   ) {variable=Is crouch}
   RigUnit_GetRelativeTransformForItem = &RigUnit_GetRelativeTransformForItem(Child=(Type=Bone,Name="head"), Parent=(Type=Bone,Name="spine_05"), bChildInitial=False, bParentInitial=False)
   RigVMFunction_MathQuaternionToEuler = &RigVMFunction_MathQuaternionToEuler(RotationOrder=ZYX, Value=@RigUnit_GetRelativeTransformForItem.RelativeTransform.Rotation)
   > RigVMVariableNode(Value=&RigVMFunction_MathDoubleRemap(SourceMaximum=-21, SourceMinimum=-10, TargetMaximum=1, TargetMinimum=0, Value=@RigVMFunction_MathQuaternionToEuler.Result.Y).Result) {variable=Head Right}
   Remap = &RigVMFunction_MathDoubleRemap(SourceMaximum=-45, SourceMinimum=-15, TargetMaximum=1, TargetMinimum=0, Value=@RigVMFunction_MathQuaternionToEuler.Result.Z)
   > RigVMVariableNode(Value=@Remap.Result) {variable=Head Up Alpha 1}
   > RigVMVariableNode(Value=&RigVMFunction_MathDoubleRemap(SourceMaximum=-55, SourceMinimum=-40, TargetMaximum=1, TargetMinimum=0, Value=@RigVMFunction_MathQuaternionToEuler.Result.Z).Result) {variable=Head Up Alpha 2}
   > RigVMVariableNode(Value=&RigVMFunction_MathDoubleRemap(SourceMaximum=-75, SourceMinimum=-65, TargetMaximum=0, TargetMinimum=1, Value=@RigVMFunction_MathQuaternionToEuler.Result.Z).Result) {variable=Head Up Alpha 3}
   n1 = &RigVMFunction_MathDoubleMul(A=RigVMVariableNode() {variable=Head Up Alpha 1}.Value, B=RigVMVariableNode() {variable=Head Up Alpha 3}.Value)
   AlphaInterp = &RigVMFunction_AlphaInterp(InterpSpeedDecreasing=25, InterpSpeedIncreasing=25, Scale=1, bClampResult=False, Value=@n1.Result)
   RigVMFunction_MathTransformFromSRT = /Script/RigVM.RigVMFunction_MathTransformFromSRT(Location=(X=0,Y=0,Z=0), Rotation=(X=0,Y=-50,Z=0), RotationOrder=XYZ, Scale=(X=1,Y=1,Z=1)) {is=struct RigVMUnitNode}
   > &RigUnit_ModifyTransforms(
    ItemToModify="((Item=(Type=Bone,Name=\"dyn_hair_back_04_Base\"),Transform=(Rotation=(X=0,Y=0,Z=0,W=1),Translation=(X=0,Y=0,Z=0.6375),Scale3D=(X=1,Y=1,Z=1))),(Item=(Type=Bone,Name=\"dyn_hair_back_05_Base\"),Transform=(Rotation=(X=0,Y=0,Z=0,W=1),Translation=(X=-1.0483,Y=0,Z=1 ...(truncated)"
    WeightMinimum=0
    Weight=@AlphaInterp.Result
    ItemToModify.0.Transform=@RigVMFunction_MathTransformFromSRT.Transform
    ItemToModify.1.Transform=@RigVMFunction_MathTransformFromSRT.Transform
    ItemToModify.2.Transform=@RigVMFunction_MathTransformFromSRT.Transform
    ItemToModify.3.Transform=@RigVMFunction_MathTransformFromSRT.Transform
   )
   Multiply = &RigVMFunction_MathDoubleMul(A=@n1.Result, B=RigVMVariableNode() {variable=Is crouch}.Value)
   > &RigUnit_ModifyTransforms(
    ItemToModify="((Item=(Type=Bone,Name=\"dyn_hair_back_04_Base\"),Transform=(Rotation=(X=0,Y=0,Z=0,W=1),Translation=(X=2,Y=0,Z=0),Scale3D=(X=1,Y=1,Z=1))),(Item=(Type=Bone,Name=\"dyn_hair_back_05_Base\"),Transform=(Rotation=(X=0,Y=0,Z=0,W=1),Translation=(X=2.3,Y=0,Z=0),Scale3D ...(truncated)"
    WeightMinimum=0
    ItemToModify.4.Transform=&RigVMFunction_MathTransformFromSRT(Location=(X=0,Y=0,Z=0), Rotation=(X=9,Y=-30,Z=-30), RotationOrder=XYZ, Scale=(X=1,Y=1,Z=1)).Transform
    Weight=@Multiply.Result
   )
   n2 = &RigVMFunction_AlphaInterp(
    InterpSpeedDecreasing=5
    InterpSpeedIncreasing=5
    Scale=1
    bClampResult=False
    Value=&RigVMFunction_MathDoubleMul(B=RigVMVariableNode() {variable=Head Up Alpha 3}.Value, A=RigVMVariableNode() {variable=Head Up Alpha 2}.Value).Result
   )
   > &RigUnit_ModifyTransforms(
    ItemToModify="((Item=(Type=Bone,Name=\"dyn_hair_back_05_Base\"),Transform=(Rotation=(X=0,Y=0,Z=0,W=1),Translation=(X=0,Y=0,Z=1),Scale3D=(X=1,Y=1,Z=1))),(Item=(Type=Bone,Name=\"dyn_hair_back_04_Base\"),Transform=(Rotation=(X=0,Y=0,Z=0,W=1),Translation=(X=0,Y=0,Z=2),Scale3D=( ...(truncated)"
    WeightMinimum=0
    Weight=@n2.Result
   )
   RigVMFunction_MathQuaternionFromEuler = &RigVMFunction_MathQuaternionFromEuler(Euler=(X=30,Y=0,Z=0))
   > &RigUnit_ModifyTransforms(
    ItemToModify="((Item=(Type=Bone,Name=\"dyn_hair_back_05_Base\"),Transform=(Rotation=(X=0,Y=0,Z=0,W=1),Translation=(X=0,Y=-2,Z=-2),Scale3D=(X=1,Y=1,Z=1))),(Item=(Type=Bone,Name=\"dyn_hair_back_06_Base\"),Transform=(Rotation=(X=0,Y=0,Z=0,W=1),Translation=(X=0,Y=0,Z=0),Scale3D ...(truncated)"
    WeightMinimum=0
    Weight=&RigVMFunction_MathDoubleRemap(SourceMaximum=1, SourceMinimum=0, TargetMaximum=0, TargetMinimum=1, Value=RigVMVariableNode() {variable=Head Up Alpha 3}.Value).Result
    ItemToModify.1.Transform.Rotation=@RigVMFunction_MathQuaternionFromEuler.Result
   )
  ]}
]
hierarchy [
 root {loc=<0 0 0>, rot=<0 0 0 1>}
 pelvis {parent=root, loc=<0 -0.5301 98.0571>, rot=<0.0127 0.707 -0.0127 -0.707>}
 spine_01 {parent=pelvis, loc=<6.1208 -0.9328 0>, rot=<0 0 0.034 -0.9994>}
 spine_02 {parent=spine_01, loc=<6.5996 -0.2095 0>, rot=<0 0 0.0238 -0.9997>}
 spine_03 {parent=spine_02, loc=<6.4914 -0.1237 0>, rot=<0 0 0.0149 -0.9999>}
 spine_04 {parent=spine_03, loc=<6.9598 0.0794 0>, rot=<0 0 -0.0132 -0.9999>}
 spine_05 {parent=spine_04, loc=<6.9987 0 0>, rot=<0 0 -0.0015 -1>}
 neck_01 {parent=spine_05, loc=<11.3138 0.1529 0>, rot=<0 0 -0.2003 -0.9797>}
 neck_02 {parent=neck_01, loc=<4.5627 0 0>, rot=<0 0 0.0504 -0.9987>}
 head {parent=neck_02, loc=<5.3914 0 0>, rot=<0 0 0.115 -0.9934>}
 dyn_hair_locan_r_master {parent=head, loc=<3.8014 3.3466 7.3203>, rot=<-0.0015 0.0697 0.9974 0.0161>}
 dyn_hair_locan_r_01 {parent=dyn_hair_locan_r_master, loc=<0.0001 0 0>, rot=<0 0 0 1>}
 dyn_hair_locan_r_02 {parent=dyn_hair_locan_r_01, loc=<4.2796 0 0>, rot=<0 -0.2042 0 -0.9789>}
 dyn_hair_locan_r_03 {parent=dyn_hair_locan_r_02, loc=<4.6879 0 0>, rot=<0 0 0 1>}
 dyn_hair_locan_f_03 {parent=head, loc=<10.9576 8.0023 1.0282>, rot=<0.3731 0.3949 0.8242 0.16>}
 dyn_hair_locan_f_04 {parent=dyn_hair_locan_f_03, loc=<2.8132 0 0>, rot=<0 -0.2883 0 -0.9575>}
 dyn_hair_locan_f_05 {parent=dyn_hair_locan_f_04, loc=<3.1712 0 0>, rot=<0 -0.1399 0 -0.9902>}
 dyn_hair_locan_f_02 {parent=head, loc=<10.2963 8.2155 -0.6785>, rot=<-0.0254 -0.8587 -0.4457 -0.2517>}
 dyn_hair_locan_f_01 {parent=dyn_hair_locan_f_02, loc=<3.2576 0 0>, rot=<0 -0.1785 0 -0.9839>}
 dyn_hair_locan_l_01 {parent=head, loc=<2.1915 -0.8126 -8.6077>, rot=<0.2121 0.604 0.7667 0.0481>}
 dyn_hair_locan_l_02 {parent=dyn_hair_locan_l_01, loc=<6.4099 0 0>, rot=<0 0 -0.0531 -0.9986>}
 dyn_hair_locan_l_03 {parent=head, loc=<4.3223 3.9275 -6.8424>, rot=<0.0768 0.5375 0.8384 0.0477>}
 dyn_hair_locan_l_04 {parent=dyn_hair_locan_l_03, loc=<3.2472 0 0>, rot=<0 0 -0.1071 -0.9942>}
 dyn_ring_r {parent=head, loc=<0.1428 0.5692 -6.0309>, rot=<0.0039 -0.9397 0.3419 -0.0014>}
 dyn_ring_l {parent=head, loc=<0.1405 0.5376 5.9579>, rot=<-0.0019 -0.8987 0.4385 0.0084>}
 Hair_back_corrective_master {parent=head, loc=<1.7488 -7.3477 0>, rot=<0 0 0 1>}
 Hair_back_corrective {parent=Hair_back_corrective_master, loc=<0 0 0>, rot=<0 0 0 1>}
 dyn_hair_back_01_Base {parent=head, loc=<5.8909 -0.7822 8.6789>, rot=<0.6858 -0.1723 0.6843 -0.178>}
 dyn_hair_back_01 {parent=dyn_hair_back_01_Base, loc=<0 0 0>, rot=<0 0 0 1>}
 dyn_hair_back_02_Base {parent=head, loc=<3.3044 -3.4464 8.8953>, rot=<0.6858 -0.1723 0.6843 -0.178>}
 dyn_hair_back_02 {parent=dyn_hair_back_02_Base, loc=<0 0 0>, rot=<0 0 0 1>}
 dyn_hair_back_03_Base {parent=head, loc=<3.3192 -4.5848 5.6736>, rot=<0.6759 -0.2079 0.6741 -0.2134>}
 dyn_hair_back_03 {parent=dyn_hair_back_03_Base, loc=<1.0085 0 -3.2397>, rot=<0 0 0 1>}
 dyn_hair_back_04_Base {parent=head, loc=<6.0018 -6.5741 0.1226>, rot=<0.4823 -0.5171 0.478 -0.5211>}
 dyn_hair_back_04 {parent=dyn_hair_back_04_Base, loc=<1.7759 0 -3.1807>, rot=<0 0 0 1>}
 dyn_hair_back_05_Base {parent=head, loc=<3.9043 -5.9539 -3.1642>, rot=<0.3737 -0.6003 0.3687 -0.6034>}
 dyn_hair_back_05 {parent=dyn_hair_back_05_Base, loc=<2.588 0 -2.5558>, rot=<0 0 0 1>}
 dyn_hair_back_06_Base {parent=head, loc=<4.8134 -2.6744 -5.9088>, rot=<0.1746 -0.6852 0.1689 -0.6866>}
 dyn_hair_back_06 {parent=dyn_hair_back_06_Base, loc=<2.3324 0 -3.492>, rot=<0 0 0 1>}
 dyn_hair_back_07_Base {parent=head, loc=<5.4023 -6.7703 3.3898>, rot=<0.3643 -0.4193 0.8013 -0.2223>}
 dyn_hair_back_07 {parent=dyn_hair_back_07_Base, loc=<2.625 0 -0.6696>, rot=<0 0 0 1>}
]
variables [
 Head Right {type=double, default=0}
 Head Up Alpha 1 {type=double, default=0}
 Head Up Alpha 2 {type=double, default=0}
 Head Up Alpha 3 {type=double, default=0}
 Is crouch {type=double, default=0}
]
