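# Physics asset for the F_MED_FairyFlex face accessory, defining capsule bodies and constraints for dynamic hair bones around the head and neck.
# Seven back pieces attach to the head through small Base bodies with free angular joints, then swing within wider limits, except back_07, which is fully free.
# Front, left and right "locan" strands form tightly limited chains; the right hangs from an angularly locked master and branches at r_01.
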PhysicsAsset /BRCosmetics/Characters/Player/Female/Medium/Bodies/F_MED_FairyFlex/Meshes/Parts/F_MED_FairyFlex_FaceAcc_Physics
bodies {
 dyn_hair_back_01 [{shapeType=Capsule, length=1.7876, radius=1.3066, shapeName=dyn_hair_back_01_capsule, center=<1.0856 -0.3325 0.4603>}]
 dyn_hair_back_01_Base [{shapeType=Capsule, length=0.505, radius=0.505, shapeName=dyn_hair_back_01_Base_capsule, rotation=<rot 0 -90 0>}]
 dyn_hair_back_02 [{shapeType=Capsule, length=2.566, radius=1.3327, shapeName=dyn_hair_back_02_capsule, center=<1.0856 0.2367 -0.3069>}]
 dyn_hair_back_02_Base [{shapeType=Capsule, length=0.505, radius=0.505, shapeName=dyn_hair_back_02_Base_capsule, rotation=<rot 0 -90 0>}]
 dyn_hair_back_03 [{shapeType=Capsule, length=2.8903, radius=2.555, shapeName=dyn_hair_back_03_capsule, center=<1.4106 0.0651 -0.0066>}]
 dyn_hair_back_03_Base [{shapeType=Capsule, length=0.505, radius=0.505, shapeName=dyn_hair_back_03_Base_capsule, rotation=<rot 0 -90 0>}]
 dyn_hair_back_04 [{shapeType=Capsule, length=4.0346, radius=2.3514, shapeName=dyn_hair_back_04_capsule, center=<1.2866 0.4082 -1.0197>}]
 dyn_hair_back_04_Base [{shapeType=Capsule, length=0.505, radius=0.505, shapeName=dyn_hair_back_04_Base_capsule, rotation=<rot 0 -90 0>}]
 dyn_hair_back_05 [{shapeType=Capsule, length=3.9433, radius=3.0926, shapeName=dyn_hair_back_05_capsule, center=<1.354 0 -0.4108>}]
 dyn_hair_back_05_Base [{shapeType=Capsule, length=0.505, radius=0.505, shapeName=dyn_hair_back_05_Base_capsule, rotation=<rot 0 -90 0>}]
 dyn_hair_back_06 [{shapeType=Capsule, length=2.7932, radius=2.5747, shapeName=dyn_hair_back_06_capsule, center=<1.4493 0 1.3531>}]
 dyn_hair_back_06_Base [{shapeType=Capsule, length=0.505, radius=0.505, shapeName=dyn_hair_back_06_Base_capsule, rotation=<rot 0 -90 0>}]
 dyn_hair_back_07 [{
   shapeType=Capsule
   length=2.8936
   radius=2.3272
   shapeName=dyn_hair_back_07_capsule
   center=<1.7814 0.7833 -0.6374>
   rotation=<rot 0 -149.5267 -79.6877>
  }]
 dyn_hair_back_07_Base [{shapeType=Capsule, length=0.505, radius=0.505, shapeName=dyn_hair_back_07_Base_capsule, rotation=<rot 0 -90 0>}]
 dyn_hair_locan_f_01 [{
   shapeType=Capsule
   length=3.3721
   radius=0.5997
   shapeName=dyn_hair_locan_f_01_capsule
   center=<1.8837 0 -0.1251>
   rotation=<rot 83.3028 0 0>
  }]
 dyn_hair_locan_f_02 [{
   shapeType=Capsule
   length=3.59
   radius=0.5997
   shapeName=dyn_hair_locan_f_02_capsule
   center=<0.9985 0 0.0736>
   rotation=<rot 85.7825 -179.9998 -179.9998>
  }]
 dyn_hair_locan_f_03 [{shapeType=Capsule, length=2.5896, radius=0.5607, shapeName=dyn_hair_locan_f_03_capsule, center=<0.6575 0 0.2295>, rotation=<rot 90 0 0>}]
 dyn_hair_locan_f_04 [{
   shapeType=Capsule
   length=3.4625
   radius=0.5607
   shapeName=dyn_hair_locan_f_04_capsule
   center=<1.2995 0 -0.0109>
   rotation=<rot 88.1533 179.9998 179.9998>
  }]
 dyn_hair_locan_f_05 [{
   shapeType=Capsule
   length=3.0757
   radius=0.5607
   shapeName=dyn_hair_locan_f_05_capsule
   center=<1.9834 0 -0.1347>
   rotation=<rot 81.9378 0 0>
  }]
 dyn_hair_locan_l_01 [{
   shapeType=Capsule
   length=5.6472
   radius=1.4494
   shapeName=dyn_hair_locan_l_01_capsule
   center=<1.5562 -0.0708 0>
   rotation=<rot 90 180 -177.3936>
  }]
 dyn_hair_locan_l_02 [{shapeType=Capsule, length=6.3041, radius=1.2241, shapeName=dyn_hair_locan_l_02_capsule, center=<2.356 0 0>, rotation=<rot 90 0 0>}]
 dyn_hair_locan_l_03 [{shapeType=Capsule, length=3.6839, radius=1.3161, shapeName=dyn_hair_locan_l_03_capsule, center=<0 -0.4675 0>, rotation=<rot 90 0 0>}]
 dyn_hair_locan_l_04 [{
   shapeType=Capsule
   length=4.17
   radius=1.2379
   shapeName=dyn_hair_locan_l_04_capsule
   center=<1.2077 0 -0.2357>
   rotation=<rot 67.3505 180 180>
  }]
 dyn_hair_locan_r_01 [{shapeType=Capsule, length=3.5, radius=1.5, shapeName=dyn_hair_locan_r_01_capsule, center=<0 0 0.2752>, rotation=<rot 90 0 0>}]
 dyn_hair_locan_r_02 [{
   shapeType=Capsule
   length=3.7051
   radius=1.2949
   shapeName=dyn_hair_locan_r_02_capsule
   center=<-0.0076 0 0.1805>
   rotation=<rot 87.5882 -179.9997 -179.9997>
  }]
 dyn_hair_locan_r_03 [{
   shapeType=Capsule
   length=6.4154
   radius=1.1576
   shapeName=dyn_hair_locan_r_03_capsule
   center=<2.3526 0 -0.2569>
   rotation=<rot 83.7678 -0.0002 -0.0002>
  }]
 dyn_hair_locan_r_04 [{
   shapeType=Capsule
   length=1.3036
   radius=0.5214
   shapeName=dyn_hair_locan_r_04_capsule
   center=<0.9659 0.0345 -0.0403>
   rotation=<rot 0 99.4883 -86.9179>
  }]
 dyn_hair_locan_r_05 [{
   shapeType=Capsule
   length=0.505
   radius=0.505
   shapeName=dyn_hair_locan_r_05_capsule
   center=<1.1654 -0.0283 -0.3149>
   rotation=<rot 0 -93.9367 -78.1418>
  }]
 dyn_hair_locan_r_master [{shapeType=Capsule, length=1.6842, radius=0.5, shapeName=dyn_hair_locan_r_master_capsule, center=<0 0 0.1172>, rotation=<rot 90 0 0>}]
 head [
  {shapeType=Capsule, length=3.8965, radius=5.8522, center=<3.3425 3.5392 0>, rotation=<rot 90 -179.1048 -173.3028>}
  {shapeType=Capsule, length=6.8514, radius=4.8393, shapeName=head_capsule, center=<0.756 4.3931 0>, rotation=<rot 90 -179.1048 -173.9807>}
 ]
 neck_01 [{shapeType=Capsule, length=8.3879, radius=4, shapeName=neck_01_capsule, center=<3.5933 0.2438 0>, rotation=<rot 90 180 -172.1134>}]
}
constraints [
 {bone1Name=dyn_hair_locan_f_02, bone2Name=head, swing1LimitDegrees=15, swing2LimitDegrees=15, twistLimitDegrees=0}
 {bone1Name=dyn_hair_locan_f_01, bone2Name=dyn_hair_locan_f_02, swing1LimitDegrees=10, swing2LimitDegrees=10, twistLimitDegrees=0}
 {bone1Name=dyn_hair_locan_f_03, bone2Name=head, swing1LimitDegrees=13, swing2LimitDegrees=13, twistLimitDegrees=0}
 {bone1Name=dyn_hair_locan_f_04, bone2Name=dyn_hair_locan_f_03, swing1LimitDegrees=7, swing2LimitDegrees=7, twistLimitDegrees=0}
 {bone1Name=dyn_hair_locan_f_05, bone2Name=dyn_hair_locan_f_04, swing1LimitDegrees=5, swing2LimitDegrees=5, twistLimitDegrees=0}
 {bone1Name=dyn_hair_locan_l_01, bone2Name=head, swing1LimitDegrees=15, swing2LimitDegrees=15, twistLimitDegrees=0}
 {bone1Name=dyn_hair_locan_l_02, bone2Name=dyn_hair_locan_l_01, swing1LimitDegrees=15, swing2LimitDegrees=15, twistLimitDegrees=0}
 {bone1Name=dyn_hair_locan_l_03, bone2Name=head, swing1LimitDegrees=15, swing2LimitDegrees=15, twistLimitDegrees=0}
 {bone1Name=dyn_hair_locan_l_04, bone2Name=dyn_hair_locan_l_03, swing1LimitDegrees=15, swing2LimitDegrees=15, twistLimitDegrees=0}
 {
  bone1Name=dyn_hair_locan_r_master
  bone2Name=head
  swing1LimitDegrees=25
  swing1Motion=Locked
  swing2LimitDegrees=25
  swing2Motion=Locked
  twistLimitDegrees=0
  twistMotion=Locked
 }
 {bone1Name=dyn_hair_locan_r_01, bone2Name=dyn_hair_locan_r_master, swing1LimitDegrees=15, swing2LimitDegrees=15, twistLimitDegrees=0}
 {bone1Name=dyn_hair_locan_r_02, bone2Name=dyn_hair_locan_r_01, swing1LimitDegrees=15, swing2LimitDegrees=15, twistLimitDegrees=0}
 {bone1Name=dyn_hair_locan_r_03, bone2Name=dyn_hair_locan_r_02, swing1LimitDegrees=15, swing2LimitDegrees=15, twistLimitDegrees=0}
 {
  bone1Name=head
  bone2Name=neck_01
  swing1LimitDegrees=45
  swing1Motion=Free
  swing2LimitDegrees=45
  swing2Motion=Free
  twistLimitDegrees=45
  twistMotion=Free
 }
 {
  bone1Name=dyn_hair_back_07
  bone2Name=dyn_hair_back_07_Base
  swing1LimitDegrees=45
  swing1Motion=Free
  swing2LimitDegrees=45
  swing2Motion=Free
  twistLimitDegrees=45
  twistMotion=Free
 }
 {bone1Name=dyn_hair_back_01, bone2Name=dyn_hair_back_01_Base, swing1LimitDegrees=50, swing2LimitDegrees=50, twistLimitDegrees=0}
 {
  bone1Name=dyn_hair_back_02_Base
  bone2Name=head
  swing1LimitDegrees=45
  swing1Motion=Free
  swing2LimitDegrees=45
  swing2Motion=Free
  twistLimitDegrees=45
  twistMotion=Free
 }
 {bone1Name=dyn_hair_back_02, bone2Name=dyn_hair_back_02_Base, swing1LimitDegrees=50, swing2LimitDegrees=50, twistLimitDegrees=0}
 {
  bone1Name=dyn_hair_back_03_Base
  bone2Name=head
  swing1LimitDegrees=45
  swing1Motion=Free
  swing2LimitDegrees=45
  swing2Motion=Free
  twistLimitDegrees=45
  twistMotion=Free
 }
 {bone1Name=dyn_hair_back_03, bone2Name=dyn_hair_back_03_Base, swing1LimitDegrees=50, swing2LimitDegrees=50, twistLimitDegrees=0}
 {
  bone1Name=dyn_hair_back_04_Base
  bone2Name=head
  swing1LimitDegrees=45
  swing1Motion=Free
  swing2LimitDegrees=45
  swing2Motion=Free
  twistLimitDegrees=45
  twistMotion=Free
 }
 {bone1Name=dyn_hair_back_04, bone2Name=dyn_hair_back_04_Base, swing1LimitDegrees=50, swing2LimitDegrees=50, twistLimitDegrees=0}
 {
  bone1Name=dyn_hair_back_05_Base
  bone2Name=head
  swing1LimitDegrees=45
  swing1Motion=Free
  swing2LimitDegrees=45
  swing2Motion=Free
  twistLimitDegrees=45
  twistMotion=Free
 }
 {bone1Name=dyn_hair_back_05, bone2Name=dyn_hair_back_05_Base, swing1LimitDegrees=50, swing2LimitDegrees=50, twistLimitDegrees=0}
 {
  bone1Name=dyn_hair_back_06_Base
  bone2Name=head
  swing1LimitDegrees=45
  swing1Motion=Free
  swing2LimitDegrees=45
  swing2Motion=Free
  twistLimitDegrees=45
  twistMotion=Free
 }
 {bone1Name=dyn_hair_back_06, bone2Name=dyn_hair_back_06_Base, swing1LimitDegrees=50, swing2LimitDegrees=50, twistLimitDegrees=0}
 {
  bone1Name=dyn_hair_back_07_Base
  bone2Name=head
  swing1LimitDegrees=45
  swing1Motion=Free
  swing2LimitDegrees=45
  swing2Motion=Free
  twistLimitDegrees=45
  twistMotion=Free
 }
 {
  bone1Name=dyn_hair_back_01_Base
  bone2Name=head
  swing1LimitDegrees=45
  swing1Motion=Free
  swing2LimitDegrees=45
  swing2Motion=Free
  twistLimitDegrees=45
  twistMotion=Free
 }
 {bone1Name=dyn_hair_locan_r_05, bone2Name=dyn_hair_locan_r_04, swing1LimitDegrees=15, swing2LimitDegrees=15, twistLimitDegrees=0}
 {bone1Name=dyn_hair_locan_r_04, bone2Name=dyn_hair_locan_r_01, swing1LimitDegrees=15, swing2LimitDegrees=15, twistLimitDegrees=0}
]
